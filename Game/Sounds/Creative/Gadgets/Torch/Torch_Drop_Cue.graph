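# Sound cue for the Creative torch gadget's drop sound, built from reused impact, trap and weapon sounds.
# A mixer layers a small wood shatter, a shotgun drop, and a briefly delayed, pitched-up flame-trap extinguish.
# An enveloper at the root fades the mix out in about a second and slightly varies pitch and volume.
# It uses weapon-drop attenuation, the loot-collect sound class and a PreventNew concurrency override.
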SoundCue /Game/Sounds/Creative/Gadgets/Torch/Torch_Drop_Cue
root=Enveloper_0
nodes [
 {id=Delay_0, children [WavePlayer_3], properties {delayMax=0.1, delayMin=0.1}}
 {
  id=Enveloper_0
  children [Mixer_16]
  properties {
   pitchMax=1.05
   pitchMin=0.95
   volumeMax=1.1
   volumeMin=0.9
   pitchCurve {editorCurveData {keys [{time=0}]}}
   volumeCurve {editorCurveData {keys [
      <at 0 1>
      {arriveTangent=-6.7251, interpMode=RCIM_Cubic, leaveTangent=-0.1336, tangentMode=RCTM_Break, time=0.4934, value=0.9833}
      <at 1.25 0>
     ]}}
  }
 }
 {id=Mixer_16, children [WavePlayer_139, Modulator_1, WavePlayer_4], properties {inputVolume [0.6,1.2,1]}}
 {id=Modulator_1, children [Delay_0], properties {pitchMax=1.35, pitchMin=1.3}}
 {id=WavePlayer_139, properties {soundWaveAssetPtr=/Game/Sounds/Fort_Impact_Sounds/Demolition/Wood/Wood_Small_Shatter01}}
 {id=WavePlayer_3, properties {soundWaveAssetPtr=/Game/Sounds/Fort_Traps/Flame_Trap/Hot_Feet_Extinguish_01}}
 {id=WavePlayer_4, properties {soundWaveAssetPtr=/Game/Sounds/Fort_Weapons/Shotgun/Combat/Shotgun_Combat_Drop01}}
]
properties {
 attenuationSettings=&Fort_Weapon_Drop_Att
 bOverrideConcurrency=true
 maxDistance=3000
 soundClassObject=&Fort_Loot_Collect
 volumeMultiplier=0.398
 concurrencyOverrides {resolutionRule=PreventNew}
}
