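# Sound cue for the looping fire of a creative torch gadget, played in the ambient sound class.
# A random node picks one of four equally weighted looping fire waves, three of them from the shared ambient fire loops, and a modulator randomizes its pitch.
# A continuous modulator at the root maps a Speed parameter to pitch and volume.
# Overridden attenuation adds HRTF spatialization, distance falloff and occlusion damping.
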SoundCue /Game/Sounds/Creative/Gadgets/Torch/Torch_Fire_Loop_Cue
root=ModulatorContinuous_0
nodes [
 {
  id=ModulatorContinuous_0
  children [Modulator_0]
  properties {
   pitchModulationParams {maxInput=1.1, maxOutput=1.1, minInput=0.95, minOutput=0.95, parameterName=Speed}
   volumeModulationParams {default=0.6, maxInput=1.5, maxOutput=1.5, minInput=0.6, minOutput=0.6, parameterName=Speed}
  }
 }
 {id=Modulator_0, children [Random_0], properties {pitchMax=1.1, pitchMin=0.9}}
 {id=Random_0, children [WavePlayer_1,WavePlayer_0,WavePlayer_2,WavePlayer_3], properties {weights [1,1,1,1]}}
 {id=WavePlayer_0, properties {bLooping=true, soundWaveAssetPtr=/Game/Sounds/Trans_AmbientLoops/Fire/Fire_Torch_03}}
 {id=WavePlayer_1, properties {bLooping=true, soundWaveAssetPtr=&/Torch_Fire_Loop01}}
 {id=WavePlayer_2, properties {bLooping=true, soundWaveAssetPtr=/Game/Sounds/Trans_AmbientLoops/Fire/Fire_Torch_02}}
 {id=WavePlayer_3, properties {bLooping=true, soundWaveAssetPtr=/Game/Sounds/Trans_AmbientLoops/Fire/Fire_Torch_01}}
]
properties {
 bOverrideAttenuation=true
 maxDistance=1450
 soundClassObject=&Fort_Ambient_Athena
 volumeMultiplier=0.832
 attenuationOverrides {
  falloffDistance=1200
  occlusionVolumeAttenuation=0.3
  spatializationAlgorithm=SPATIALIZATION_HRTF
  stereoSpread=0
  attenuationShapeExtents {x=250}
 }
}
