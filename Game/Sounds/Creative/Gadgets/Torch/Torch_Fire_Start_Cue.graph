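# Sound cue for the Creative torch gadget's fire-start sound, routed to the Athena ambient sound class.
# A random node picks one of four equally weighted start waves and passes it through a modulator at the root.
# The cue overrides attenuation for distance falloff with occlusion filtering and volume reduction, and overrides concurrency to cap simultaneous instances.
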SoundCue /Game/Sounds/Creative/Gadgets/Torch/Torch_Fire_Start_Cue
root=Modulator_0
nodes [
 {id=Modulator_0, children [Random_0]}
 {
  id=Random_0
  children [WavePlayer_0,WavePlayer_1,WavePlayer_2,WavePlayer_3]
  properties {bShouldExcludeFromBranchCulling=true, weights [1,1,1,1]}
 }
 {id=WavePlayer_0, properties {soundWaveAssetPtr=/Game/Sounds/Creative/Gadgets/Torch/New/Torch_Fire_Start_New_01}}
 {id=WavePlayer_1, properties {soundWaveAssetPtr=/Game/Sounds/Creative/Gadgets/Torch/New/Torch_Fire_Start_New_02}}
 {id=WavePlayer_2, properties {soundWaveAssetPtr=/Game/Sounds/Creative/Gadgets/Torch/New/Torch_Fire_Start_New_03}}
 {id=WavePlayer_3, properties {soundWaveAssetPtr=/Game/Sounds/Creative/Gadgets/Torch/New/Torch_Fire_Start_New_04}}
]
properties {
 bOverrideAttenuation=true
 bOverrideConcurrency=true
 maxDistance=1700
 soundClassObject=&Fort_Ambient_Athena
 volumeMultiplier=1.253
 attenuationOverrides {
  dBAttenuationAtMax=-40
  falloffDistance=1500
  occlusionLowPassFilterFrequency=10000
  occlusionVolumeAttenuation=0.3
  stereoSpread=0
  attenuationShapeExtents {x=200}
 }
 concurrencyOverrides {maxCount=6}
}
