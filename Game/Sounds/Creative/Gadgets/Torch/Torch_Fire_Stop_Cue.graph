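# Sound cue for the Creative torch gadget, played, going by its name, when the torch fire stops.
# A single stop wave runs through a default-setting Modulator, with cue-level volume boosted and pitch lowered slightly, in the Athena ambient sound class.
# Overridden attenuation adds distance falloff and occlusion with low-pass filtering and volume reduction on a game trace channel.
# A concurrency override caps simultaneous instances.
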SoundCue /Game/Sounds/Creative/Gadgets/Torch/Torch_Fire_Stop_Cue
root=Modulator_0
nodes [
 {id=Modulator_0, children [WavePlayer_0]}
 {id=WavePlayer_0, properties {soundWaveAssetPtr=&/Torch_Fire_Stop01}}
]
properties {
 bOverrideAttenuation=true
 bOverrideConcurrency=true
 maxDistance=1700
 pitchMultiplier=0.9
 soundClassObject=&Fort_Ambient_Athena
 volumeMultiplier=2.079
 attenuationOverrides {
  bEnableOcclusion=true
  dBAttenuationAtMax=-40
  falloffDistance=1500
  occlusionLowPassFilterFrequency=10000
  occlusionTraceChannel=ECC_GameTraceChannel11
  occlusionVolumeAttenuation=0.5
  stereoSpread=0
  attenuationShapeExtents {x=200}
 }
 concurrencyOverrides {maxCount=6}
}
