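# Sound cue played when the Creative torch gadget is picked up, routed to the loot-collect class with weapon-pickup attenuation.
# A mixer layers a wood shatter faded by a volume envelope, a slightly delayed torch-ignition sound as the loudest layer, and a later, quiet, pitched-up dragon fire burst.
# A root modulator adds default random pitch and volume variation to the mix, and a shared pickup concurrency set limits overlapping instances.
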SoundCue /Game/Sounds/Creative/Gadgets/Torch/Torch_Pickup_Cue
root=Modulator_2
nodes [
 {id=Delay_2, children [WavePlayer_6], properties {delayMax=0.08, delayMin=0.08}}
 {id=Delay_3, children [Modulator_3], properties {delayMax=0.25, delayMin=0.25}}
 {
  id=Enveloper_0
  children [WavePlayer_12]
  properties {
   pitchCurve {editorCurveData {keys [{time=0}]}}
   volumeCurve {editorCurveData {keys [{interpMode=RCIM_Cubic, tangentMode=RCTM_Break, time=0, value=1}, <at 1.1 0>]}}
  }
 }
 {id=Mixer_0, children [Enveloper_0,Delay_2,Delay_3], properties {inputVolume [0.3,1.2,0.2]}}
 {id=Modulator_2, children [Mixer_0]}
 {id=Modulator_3, children [WavePlayer_13], properties {pitchMax=1.5, pitchMin=1.5}}
 {id=WavePlayer_12, properties {soundWaveAssetPtr=/Game/Sounds/Fort_Impact_Sounds/Demolition/Wood/Wood_Small_Shatter05}}
 {id=WavePlayer_13, properties {soundWaveAssetPtr=/Game/Sounds/Fort_Gadgets/Dragon_Slash/Fire_of_the_Dragon_Start_01}}
 {id=WavePlayer_6, properties {soundWaveAssetPtr=&/Torch_Fire_Start01}}
]
properties {
 attenuationSettings=&Fort_Weapon_Pickup_Att
 maxDistance=3500
 soundClassObject=&Fort_Loot_Collect
 volumeMultiplier=0.697
 concurrencyOverrides {maxCount=1, resolutionRule=PreventNew}
 concurrencySet [&PickupMaximum_Oldest]
}
